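# Looping sound cue for the Tactical Fisherman cosmetic glider.
# A Mixer layers two looping wave players, with the base loop mixed well below the Loop03 layer.
# The cue runs at reduced overall volume, using the glider thrust attenuation, a glider sound class and the cosmetic glider loop audio properties sheet.
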SoundCue /BRCosmetics/Sounds/Gliders/TechnicalFisherman/Glider_TacticalFisherman_Loop_Cue
root=Mixer_0
nodes [
 {id=Mixer_0, children [WavePlayer_0,WavePlayer_2], properties {inputVolume [0.3,1]}}
 {id=WavePlayer_0, properties {bLooping=true, soundWaveAssetPtr=&/Glider_TacticalFisherman_Loop}}
 {
  id=WavePlayer_2
  properties {bLooping=true, soundWaveAssetPtr=&/Glider_TacticalFisherman_Loop03}
 }
]
properties {
 attenuationSettings=&Fort_Player_Glider_Thrust_Athena_Att
 audioPropertiesSheet=&APS_Cosmetic_Glider_Loop
 maxDistance=12500
 soundClassObject=&Fort_Gliders_Land
 volumeMultiplier=0.2
}
